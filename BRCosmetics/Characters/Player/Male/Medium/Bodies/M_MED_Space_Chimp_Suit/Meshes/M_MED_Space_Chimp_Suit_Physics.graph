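# Physics asset for the male medium Space Chimp Suit body, covering the suit's left and right straps and zippers.
# Small capsule bodies sit on the two straps, two zippers, both lower calf twist bones and the first spine bone.
# Constraints tie each zipper to its calf twist bone and each strap to spine_01.
# Twist and first swing have zero limit angles, leaving a narrow second swing for hinge-like dangling.
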PhysicsAsset /BRCosmetics/Characters/Player/Male/Medium/Bodies/M_MED_Space_Chimp_Suit/Meshes/M_MED_Space_Chimp_Suit_Physics
bodies {
 calf_twist_02_l [{shapeType=Capsule, length=2, radius=2, shapeName=calf_twist_02_l_capsule, center=<0.0547 -0.575 -0.3652>, rotation=<rot 0 96.8774 -86.6775>}]
 calf_twist_02_r [{shapeType=Capsule, length=2, radius=2, shapeName=calf_twist_02_r_capsule, center=<-0.3194 0.5098 0.2956>, rotation=<rot 0 98.3909 -87.036>}]
 spine_01 [{shapeType=Capsule, length=1, radius=1, shapeName=spine_01_capsule, center=<2.6236 -0.4841 0.0027>, rotation=<rot 0.1367 -90 -0.0934>}]
 strap_l [{shapeType=Capsule, length=2, radius=2, shapeName=strap_l_capsule, center=<0 0 0.1>}]
 strap_r [{shapeType=Capsule, length=2, radius=2, shapeName=strap_r_capsule, center=<0 0 -0.1>}]
 zipper_R [{
   shapeType=Capsule
   length=2.6131
   radius=1.1909
   shapeName=zipper_R_capsule
   center=<-2.0657 0.2228 -0.2963>
   rotation=<rot 0 84.5456 -77.9459>
  }]
 zipper_l [{
   shapeType=Capsule
   length=2.6122
   radius=1.1984
   shapeName=zipper_l_capsule
   center=<2.0692 -0.1881 0.2997>
   rotation=<rot 0 83.4176 -77.7678>
  }]
}
constraints [
 {bone1Name=zipper_R, bone2Name=calf_twist_02_r, swing1LimitDegrees=0, swing2LimitDegrees=25, twistLimitDegrees=0}
 {bone1Name=zipper_l, bone2Name=calf_twist_02_l, swing1LimitDegrees=0, swing2LimitDegrees=25, twistLimitDegrees=0}
 {bone1Name=strap_l, bone2Name=spine_01, swing1LimitDegrees=0, swing2LimitDegrees=20, twistLimitDegrees=0}
 {bone1Name=strap_r, bone2Name=spine_01, swing1LimitDegrees=0, swing2LimitDegrees=20, twistLimitDegrees=0}
]
