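# Physics asset for the face accessory mesh of the male medium Space Chimp Suit cosmetic, covering the head and a dynamic hair bone chain.
# Small, equally sized capsule bodies sit on the head and three dyn_hair bones.
# Constraints link head to dyn_hair1, dyn_hair2 and dyn_hair3 in sequence, with matching narrow swing and tighter twist limits that restrict how far the chain bends.
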PhysicsAsset /BRCosmetics/Characters/Player/Male/Medium/Bodies/M_MED_Space_Chimp_Suit/Meshes/Parts/M_MED_Space_Chimp_Suit_FaceAcc_Physics
bodies {
 dyn_hair1 [{shapeType=Capsule, length=2, radius=2, shapeName=dyn_hair1_capsule, center=<0.1 0 0>}]
 dyn_hair2 [{shapeType=Capsule, length=2, radius=2, shapeName=dyn_hair2_capsule, center=<0.1 0 0>}]
 dyn_hair3 [{shapeType=Capsule, length=2, radius=2, shapeName=dyn_hair3_capsule, center=<0.1 0 0>}]
 head [{shapeType=Capsule, length=2, radius=2, shapeName=head_capsule}]
}
constraints [
 {bone1Name=dyn_hair2, bone2Name=dyn_hair1, swing1LimitDegrees=20, swing2LimitDegrees=20, twistLimitDegrees=5}
 {bone1Name=dyn_hair3, bone2Name=dyn_hair2, swing1LimitDegrees=20, swing2LimitDegrees=20, twistLimitDegrees=5}
 {bone1Name=dyn_hair1, bone2Name=head, swing1LimitDegrees=20, swing2LimitDegrees=20, twistLimitDegrees=5}
]
